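# Translucent, two-sided lit material for animated lobby chimney smoke on spline-based geometry.
# SplineThicken sets the strip width, added to a time-driven sine sway keyed to world Z along a Direction parameter, with amplitude scaled by inverted V.
# A panning tileable smoke texture is tinted by Color and also multiplied by a UV mask that fades both side edges and falls off along V for opacity.
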Material /Game/Environments/Props/Winter/Materials/M_LobbyChimneySmoke
graph {
 properties {
  bTangentSpaceNormal=false
  blendMode=BLEND_Translucent
  pixelDepthOffsetMode=PDOM_Legacy
  translucencyLightingMode=TLM_VolumetricPerVertexNonDirectional
  twoSided=true
  usedShadingModels=Default Lit
 }
 nodes [
  n1 = VectorParameter() {parameterName=Direction, defaultValue=<rgba 1 1.5465 -0.5478 1>}
  n2 = ComponentMask(None=WorldPosition().XYZ) {b=true}
  n3 = Multiply(A=Time(), B=ScalarParameter() {defaultValue=-23.0577, parameterName=SpeedAnimation, sortPriority=1})
  n4 = Sine(None=Divide(A=Add(A=@n2, B=@n3), B=ScalarParameter() {defaultValue=317.0983, parameterName=WaveLength}))
  n5 = TextureCoordinate()
  n6 = Multiply(A=OneMinus(None=ComponentMask(None=@n5) {g=true}), B=ScalarParameter() {defaultValue=147.0195, parameterName=Amplitude})
  Add(
   A=MaterialFunctionCall(WidthBase=ScalarParameter() {parameterName=Width}, WidthTip=ScalarParameter() {defaultValue=250.3667, parameterName=Width_1}) {materialFunction=&SplineThicken}.WorldPositionOffset
   B=Multiply(A=Normalize(VectorInput=@n1.RGB), B=Multiply(A=@n4, B=@n6))
  )
  n7 = Time()
  n8 = ScalarParameter() {defaultValue=1, parameterName=Speed}
  n9 = TextureSample(UVs=Panner(Coordinate=TextureCoordinate() {vTiling=2}, Time=Multiply(A=@n7, B=@n8)) {speedY=0.1}) {samplerType=SAMPLERTYPE_LinearColor, texture=/Game/Global/Textures/Default/Tileable_Patterns/T_Tile_Smoke_Small}
  n10 = TextureCoordinate()
  n11 = ComponentMask(None=@n10) {r=true}
  n12 = Power(Base=Multiply(A=@n11, B=OneMinus(None=@n11)), Exp=ScalarParameter() {defaultValue=1.3066, parameterName=OpacityFalloff_1})
  n13 = ScalarParameter() {defaultValue=24.3278, parameterName=Opacity_1, sortPriority=1}
  n14 = Power(Base=ComponentMask(None=@n10) {g=true}, Exp=ScalarParameter() {defaultValue=1.4667, parameterName=OpacityFalloff})
  n15 = ScalarParameter() {defaultValue=0.4807, parameterName=Opacity, sortPriority=1}
  Multiply(A=@n9.RGB, B=Multiply(A=Multiply(A=@n12, B=@n13), B=Multiply(A=@n14, B=@n15)))
  Multiply(A=@n9.RGB, B=VectorParameter() {parameterName=Color, defaultValue=<rgba 2 2 2 1>}.RGB)
 ]
}
